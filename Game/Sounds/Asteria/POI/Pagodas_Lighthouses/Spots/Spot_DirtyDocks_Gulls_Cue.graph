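# Sound cue for spot gull one-shots, filed under Asteria pagoda and lighthouse POI sounds and reusing Apollo DirtyDocks biome recordings.
# A Random node picks one of five equally weighted gull waves, and a Modulator varies its pitch and volume.
# The root quality node feeds this chain to its first two levels and leaves the other two empty.
# It plays quietly as an outdoor ambient one-shot with shared tile-ambient attenuation.
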SoundCue /Game/Sounds/Asteria/POI/Pagodas_Lighthouses/Spots/Spot_DirtyDocks_Gulls_Cue
root=QualityLevel_0
nodes [
 {id=Modulator_0, children [Random_1], properties {pitchMax=1.2, pitchMin=0.8, volumeMin=0.8}}
 {id=QualityLevel_0, children [Modulator_0,Modulator_0,null,null]}
 {id=Random_1, children [WavePlayer_8, WavePlayer_9, WavePlayer_10, WavePlayer_11, WavePlayer_12], properties {weights [1,1,1,1,1]}}
 {id=WavePlayer_10, properties {soundWaveAssetPtr=/Game/Sounds/Apollo/Biomes/DirtyDocks/Biome_DirtyDocks_OneShot_Gulls_06}}
 {id=WavePlayer_11, properties {soundWaveAssetPtr=/Game/Sounds/Apollo/Biomes/DirtyDocks/Biome_DirtyDocks_OneShot_Gulls_07}}
 {id=WavePlayer_12, properties {soundWaveAssetPtr=/Game/Sounds/Apollo/Biomes/DirtyDocks/Biome_DirtyDocks_OneShot_Gulls_08}}
 {id=WavePlayer_8, properties {soundWaveAssetPtr=/Game/Sounds/Apollo/Biomes/DirtyDocks/Biome_DirtyDocks_OneShot_Gulls_04}}
 {id=WavePlayer_9, properties {soundWaveAssetPtr=/Game/Sounds/Apollo/Biomes/DirtyDocks/Biome_DirtyDocks_OneShot_Gulls_05}}
]
properties {
 attenuationSettings=/Game/Sounds/Attenuation/Fort_TileAmbient_OneShot_Att
 maxDistance=8000
 soundClassObject=&Fort_Ambient_OneShots_Outdoors
 volumeMultiplier=0.251
 preEffectBusSends [{}]
}
